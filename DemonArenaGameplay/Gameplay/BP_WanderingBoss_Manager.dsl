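; Manager for the wandering Fire Petal boss in Demon Arena that decides when the boss relocates.
; It catches the tagged boss through the Living World spawn delegate, granting storm immunity if in an arena, otherwise setting a health invulnerability floor.
; Outside arenas, safe-zone phases, arena spawns and reaching that floor send a relocate gameplay event, guarded by bIsRelocating.
; Safe-zone relocation waits while the AI is suspicious or threatened, resuming after alert drops and a delay.
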(event EventBeginPlay
  (Class|FortAsyncActionCurrentPlaylistReady|PlaylistDataReadyAsync
    (:OnReady
      (EventDispatchers|BindEventtoOnActorSpawnedDelegate (Class|FortAthenaLivingWorldManager|GetLivingWorldManager) (delegate Custom|OnActorSpawned) :!class /Script/LagerRuntime.FortAthenaLivingWorldManager))))

(event Custom|OnActorSpawned
  (params (SpawnedActor :type Actor%20Object%20Reference))
  (bind _wanderingbosspawn (Variables|InternalData|GetWanderingBossPawn))
  (bind _returnvalue (GameplayTags|HasMatchingGameplayTag(Message) SpawnedActor "(TagName=\"AISpawnerData.Type.Boss.FirePetal.WB\")"))
  (if _returnvalue
    (bind _asbp_player_pawn_non_participant_boss (Utilities|Casting|CastToBP_PlayerPawn_NonParticipant_Boss SpawnedActor))
    (bind _output_get (Variables|InternalData|SetWanderingBossPawn _asbp_player_pawn_non_participant_boss))
    (bind _asfort_athena_aibot_controller (Utilities|Casting|CastToFortAthenaAIBotController (AI|GetAIController _output_get)))
    (Variables|InternalData|SetAIController _asfort_athena_aibot_controller)
    (bind _bdidactorhavetags (Class|AthenaQuestFunctionLibrary|AthenaCheckActorHasTags (Variables|InternalData|GetWanderingBossPawn) "(GameplayTags=((TagName=\"FirePetalBosses.Demons.IsInArena\")))" true))
    (if _bdidactorhavetags
      (Development|PrintString "@@@ OnActorSpawned [BP_WanderingBoss_Manager]: CALLED - In Arena!" false (Variables|Debug|GetbDebugManager))
      (Ability|GameplayEffect|AddLooseGameplayTags (Variables|InternalData|GetWanderingBossPawn) (Variables|Tags|GetStormImmunityTag))
      (else
        (Development|PrintString "@@@ OnActorSpawned [BP_WanderingBoss_Manager]: CALLED - Not in Arena!" false (Variables|Debug|GetbDebugManager))
        (Variables|InternalData|SetbIsRelocating false)
        (Health|SetHealthInvulnerabilityRatio _wanderingbosspawn (ScalableFloat|GetValueatLevel (Variables|Tunables|GetSF_NonArena_InvulRatio)))
        (EventDispatchers|BindEventtoOnDamaged _wanderingbosspawn (EventDispatchers|CreateEvent :!event Custom|OnWanderingBossDamaged) :!class /Script/FortniteGame.FortPawn)
        (bind _returnvalue_2 (GameplayEventRouter|GetEventRouter "/Script/FortniteGame.FortGameState" self))
        (bind (_asynctaskproxy _payload) (GameplayEventRouter|StartListeningToStatefulEvent _returnvalue_2 "/Script/FortniteGame.FortSafeZonePhaseUpdatedEvent")
          (:then
            (Variables|InternalData|SetFortSafeZoneEventHandle _asynctaskproxy)
            (bind _returnvalue_3 (GameplayEventRouter|GetEventRouter "/DemonArenaGameplay/Gameplay/B_FirePetal_DemonArena_Playspace.B_FirePetal_DemonArena_Playspace_C" self))
            (bind (_asynctaskproxy_2 _payload_2) (GameplayEventRouter|StartListeningToStatefulEvent _returnvalue_3 "/DemonArenaGameplay/Gameplay/OnDemonArenaSpawnedEvent.OnDemonArenaSpawnedEvent")
              (:then
                (Variables|InternalData|SetDemonArenaSpawnEventHandle _asynctaskproxy_2))
              (:OnEventReceived
                (bind _payload_arenaplayspace_3_61f0c4a4494904e12d4019a3cd8a5153 (Utilities|Struct|BreakOnDemonArenaSpawnedEvent _payload_2))
                (CallFunction|OnDemonArenaSpawnEvent _payload_arenaplayspace_3_61f0c4a4494904e12d4019a3cd8a5153))))
          (:OnEventReceived
            (bind _payload_safezonephaseindex (Utilities|Struct|BreakFortSafeZonePhaseUpdatedEvent _payload))
            (CallFunction|OnSafeZoneEvent _payload_safezonephaseindex)))))))

(event Custom|ForceRelocateWanderer
  (Development|PrintString "@@@@@@ ForceRelocateWanderer [BP_WanderingBoss_Manager]: CALLED -force CALLED!" false (Variables|Debug|GetbDebugManager))
  (if (not (Variables|InternalData|GetbIsRelocating))
    (Variables|InternalData|SetbIsRelocating true)
    (Ability|SendGameplayEventtoActor (Variables|InternalData|GetWanderingBossPawn) "(TagName=\"Event.WanderingBoss.Relocate\")")))

(event Custom|OnSafeZoneEvent
  (params (PayloadPhaseIdx :type Integer))
  (Variables|InternalData|SetPhaseIdx PayloadPhaseIdx)
  (Development|PrintString (Utilities|String|Append "@@@@@@ On Event Received [BP_WanderingBoss_Manager]: SAFE ZONE index = " (Utilities|String|ToString(Integer) (Variables|InternalData|GetPhaseIdx)) "!") false (Variables|Debug|GetbDebugManager))
  (switch int (Variables|InternalData|GetPhaseIdx)
    (:2
      _)
    (:3
      _)
    (:4
      _))
  (switch Utilities|FlowControl|Switch|SwitchonEAlertLevel (Variables|/Script/FortniteAI.FortAthenaAIBotController|GetCurrentAlertLevel (Variables|InternalData|GetAIController))
    (:Unaware
      _)
    (:Suspicious
      (AlertLevel|BindEventtoOnAlertLevelChangedEvent (Variables|InternalData|GetAIController) (delegate Custom|OnAlertLevelChanged) :!label _label))
    (:LostTarget
      _)
    (:Threatened
      (goto _label))
    (:Count
      _))
  (if (not (Variables|InternalData|GetbIsRelocating))
    (Variables|InternalData|SetbIsRelocating true)
    (Ability|SendGameplayEventtoActor (Variables|InternalData|GetWanderingBossPawn) "(TagName=\"Event.WanderingBoss.Relocate\")")))

(event Custom|OnDemonArenaSpawnEvent
  (params (PayloadArenaPlayspace :type B%20Fire%20Petal%20Demon%20Arena%20Playspace%20Object%20Reference))
  (Development|PrintString "@@@@@@ OnDemonArenaSpawnEvent [BP_WanderingBoss_Manager]: CALLED - Arena here, I gotta go!" false (Variables|Debug|GetbDebugManager))
  (if (not (Variables|InternalData|GetbIsRelocating))
    (Variables|InternalData|SetbIsRelocating true)
    (Ability|SendGameplayEventtoActor (Variables|InternalData|GetWanderingBossPawn) "(TagName=\"Event.WanderingBoss.Relocate\")")))

(event Custom|OnAlertLevelChanged
  (params (BotController :type Fort%20Athena%20AIBot%20Controller%20Object%20Reference) (OldAlertLevel :type EAlertLevel%20Enum :default "Unaware") (NewAlertLevel :type EAlertLevel%20Enum :default "Unaware"))
  (switch Utilities|FlowControl|Switch|SwitchonEAlertLevel NewAlertLevel
    (:Unaware
      _)
    (:LostTarget
      _)
    (:Count
      _))
  (AlertLevel|UnbindEventfromOnAlertLevelChangedEvent (Variables|InternalData|GetAIController) (delegate Custom|OnAlertLevelChanged))
  (Utilities|FlowControl|Delay (ScalableFloat|GetValueatLevel (Variables|Tunables|GetSF_AlertSenseDelay)))
  (if (not (Variables|InternalData|GetbIsRelocating))
    (Variables|InternalData|SetbIsRelocating true)
    (Ability|SendGameplayEventtoActor (Variables|InternalData|GetWanderingBossPawn) "(TagName=\"Event.WanderingBoss.Relocate\")")))

(event Custom|OnWanderingBossDamaged
  (params (DamagedActor :type Actor%20Object%20Reference) (Damage :type Float%20%28single-precision%29) (InstigatedBy :type Controller%20Object%20Reference) (DamageCauser :type Actor%20Object%20Reference) (HitLocation :type Vector) (FHitComponent :type Primitive%20Component%20Object%20Reference) (BoneName :type Name) (Momentum :type Vector))
  (bind _wanderingbosspawn (Variables|InternalData|GetWanderingBossPawn))
  (if (<= (FortPawn|HealthPercent (Variables|InternalData|GetWanderingBossPawn)) (ScalableFloat|GetValueatLevel (Variables|Tunables|GetSF_NonArena_InvulRatio)))
    (Development|PrintString "@@@@@@ OnWanderingBossDamaged [BP_WanderingBoss_Manager]: CALLED - I am invulnerable!" false (Variables|Debug|GetbDebugManager))
    (if (not (Variables|InternalData|GetbIsRelocating))
      (Variables|InternalData|SetbIsRelocating true)
      (Ability|GameplayEffect|AddLooseGameplayTags _wanderingbosspawn (Variables|Tags|GetRelocateForcedTag))
      (Ability|SendGameplayEventtoActor _wanderingbosspawn "(TagName=\"Event.WanderingBoss.Relocate\")")
      (AlertLevel|UnbindEventfromOnAlertLevelChangedEvent (Variables|InternalData|GetAIController) (delegate Custom|OnAlertLevelChanged)))))

(fn ConstructionScript)
