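; Demon arena gameplay volume that tracks actors entering and leaving the arena through ArenaVolume overlap events.
; Overlaps notify the vortex volume component, with entry gated by TryDelayingEnterfromTop and exit immediate.
; On authority, FirePetal bosses tagged as in-arena gain loose exit tags on leaving and lose them on overlap.
; On remotes, pawns with a valid controller trigger local-player enter and exit events, which are empty here.
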(event Custom|OnLocalPlayerEnteredArena)

(event Custom|OnLocalPlayerExitedArena)

(event OnComponentBeginOverlap(ArenaVolume)
  (params (OverlappedComponent :type Primitive%20Component%20Object%20Reference) (OtherActor :type Actor%20Object%20Reference) (OtherComp :type Primitive%20Component%20Object%20Reference) (OtherBodyIndex :type Integer) (bFromSweep :type Boolean) (SweepResult :type Hit%20Result%20Structure%20%28by%20ref%29))
  (CallFunction|TestBossOverlap OtherActor)
  (Utilities|TryDelayingEnterfromTop OtherActor
    (:Completed
      (Class|BPDemonArenaVortexVolumeComponent|OnEnteredVolume (Variables|Default|GetBP_DemonArena_VortexVolumeComponent) OtherActor)
      (bind _asfort_pawn (Utilities|Casting|CastToFortPawn OtherActor))
      (Networking|SwitchHasAuthority
        (:Remote
          (Utilities|IsValid (Pawn|GetController _asfort_pawn)
            (:Is%20Valid
              (CallFunction|OnLocalPlayerEnteredArena))))))))

(event OnComponentEndOverlap(ArenaVolume)
  (params (OverlappedComponent :type Primitive%20Component%20Object%20Reference) (OtherActor :type Actor%20Object%20Reference) (OtherComp :type Primitive%20Component%20Object%20Reference) (OtherBodyIndex :type Integer))
  (CallFunction|TestBossEndOverlap OtherActor)
  (Class|BPDemonArenaVortexVolumeComponent|OnExitedVolume (Variables|Default|GetBP_DemonArena_VortexVolumeComponent) OtherActor)
  (bind _asfort_pawn (Utilities|Casting|CastToFortPawn OtherActor))
  (Networking|SwitchHasAuthority
    (:Remote
      (Utilities|IsValid (Pawn|GetController _asfort_pawn)
        (:Is%20Valid
          (CallFunction|OnLocalPlayerExitedArena))))))

(fn EarlyReturnAND
  (params (Cond1 :type Boolean) (Cond2 :type Boolean))
  (if Cond1
    (return Cond2)
    (else
      (return false))))

(fn TestBossEndOverlap
  (params (InTestActor :type Actor%20Object%20Reference))
  (Networking|SwitchHasAuthority
    (:Authority
      (bind _returnvalue (GameplayTags|HasMatchingGameplayTag(Message) InTestActor "(TagName=\"AISpawnerData.Type.Boss.FirePetal.WB\")"))
      (if _returnvalue
        (bind _bdidactorhavetags (Class|AthenaQuestFunctionLibrary|AthenaCheckActorHasTags InTestActor "(GameplayTags=((TagName=\"FirePetalBosses.Demons.IsInArena\")))" true))
        (if _bdidactorhavetags
          (Development|PrintString "@@@ TestBossEndOverlap [B_FirePetal_DemonArena_GameplayVolume]: Arena Boss has ended overlap!" false (Variables|Debug|GetbDebugVolume))
          (Ability|GameplayEffect|AddLooseGameplayTags (Variables|GetInTestActor) (Variables|Tags|GetBossArenaExitTags)))))))

(fn TestBossOverlap
  (params (InTestActor :type Actor%20Object%20Reference))
  (Networking|SwitchHasAuthority
    (:Authority
      (bind _returnvalue (GameplayTags|HasMatchingGameplayTag(Message) InTestActor "(TagName=\"AISpawnerData.Type.Boss.FirePetal.WB\")"))
      (if _returnvalue
        (bind _bdidactorhavetags (Class|AthenaQuestFunctionLibrary|AthenaCheckActorHasTags InTestActor "(GameplayTags=((TagName=\"FirePetalBosses.Demons.IsInArena\")))" true))
        (if _bdidactorhavetags
          (Development|PrintString "@@@ TestBossOverlap [B_FirePetal_DemonArena_GameplayVolume]: Arena Boss has overlapped!" false (Variables|Debug|GetbDebugVolume))
          (Ability|GameplayEffect|RemoveLooseGameplayTags (Variables|GetInTestActor) (Variables|Tags|GetBossArenaExitTags)))))))

(fn ConstructionScript)
