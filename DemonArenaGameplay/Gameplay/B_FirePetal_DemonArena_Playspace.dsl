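; Playspace for the FirePetal Demon Arena boss encounter, tracking the wandering boss spawned inside it.
; BeginPlay broadcasts a stateful spawned event and, on authority, registers objective auto-marking, which EndPlay unregisters.
; Once the gameplay volume binds, authority stores its position and listens to its AI point provider; ResetDemonArena repeats this setup without an authority check.
; Spawned bosses are stored and watched for death, which only prints; BossArenaAwareness is an empty stub.
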(event EventBeginPlay
  (Utilities|FlowControl|Sequence
    (:then_0
      ; Register for Auto Marking
      (Networking|SwitchHasAuthority
        (:Authority
          (Class|BMMObjObjectiveActorComponent|RegisterObjectiveforAutoMarking (Variables|Default|GetB_MMObj_ObjectiveActorComponent))
          (Networking|FlushNetDormancy))))
    (:then_1
      (GameplayEventRouter|BroadcastStatefulEvent (GameplayEventRouter|GetGlobalEventRouter) (Utilities|Struct|MakeOnDemonArenaSpawnedEvent self) 0)
      (Utilities|IsValid (Variables|Playspace|GetBoundGameplayVolume)
        (:Is%20Valid
          (CallFunction|OnBoundGameplayVolumeSet 0))
        (:Is%20Not%20Valid
          (Playspace|BindEventtoNotifyBoundGameplayVolumeSet (delegate Custom|OnBoundGameplayVolumeSet)))))))

(event Custom|OnBoundGameplayVolumeSet
  (params (Playspace :type Playspace%20Object%20Reference))
  (Networking|SwitchHasAuthority
    (:Authority
      (Utilities|FlowControl|DoOnce false
        (:Completed
          (EventDispatchers|BindEventtoOnActorSpawnedResult (Variables|/DemonArenaGameplay/Gameplay/B_FirePetal_DemonArena_GameplayVolume.B_FirePetal_DemonArena_GameplayVolume_C|GetAIPointProvider (Utilities|Casting|CastToB_FirePetal_DemonArena_GameplayVolume (Variables|Playspace|GetBoundGameplayVolume))) (EventDispatchers|CreateEvent :!event Custom|OnBossSpawned))
          (Variables|DemonArenaPlayspace|SetGameplayVolumePosition (Transformation|GetActorLocation (Variables|Playspace|GetBoundGameplayVolume))))))))

(event Custom|ResetDemonArena
  (Utilities|FlowControl|DoOnce false
    (:Completed
      (EventDispatchers|BindEventtoOnActorSpawnedResult (Variables|/DemonArenaGameplay/Gameplay/B_FirePetal_DemonArena_GameplayVolume.B_FirePetal_DemonArena_GameplayVolume_C|GetAIPointProvider (Utilities|Casting|CastToB_FirePetal_DemonArena_GameplayVolume (Variables|Playspace|GetBoundGameplayVolume))) (EventDispatchers|CreateEvent :!event Custom|OnBossSpawned))
      (Variables|DemonArenaPlayspace|SetGameplayVolumePosition (Transformation|GetActorLocation (Variables|Playspace|GetBoundGameplayVolume))))))

(event Custom|OnBossSpawned
  (params (SpawnedActor :type Actor%20Object%20Reference) (bResult :type Boolean))
  ;[0 0.2021 1 0.5] Wait for the Spawned Boss AI and process
  (bind _asbp_player_pawn_non_participant_boss (Utilities|Casting|CastToBP_PlayerPawn_NonParticipant_Boss SpawnedActor))
  (bind _output_get (Variables|InternalData|SetWanderingBossPawn _asbp_player_pawn_non_participant_boss))
  (EventDispatchers|AssignOnDied _output_get (delegate Custom|OnDied_Event))
  (CallFunction|BossArenaAwareness))

(event Custom|OnDied_Event
  (params (DamagedActor :type Actor%20Object%20Reference) (Damage :type Float%20%28single-precision%29) (InstigatedBy :type Controller%20Object%20Reference) (DamageCauser :type Actor%20Object%20Reference) (HitLocation :type Vector) (FHitComponent :type Primitive%20Component%20Object%20Reference) (BoneName :type Name) (Momentum :type Vector))
  ;[0 0.2021 1 0.5] Boss Defeated Event
  (Development|PrintString "The Boss has Perished!" true true "(R=1.000000,G=0.000000,B=0.340070,A=1.000000)" 5.0))

(event EventEndPlay
  (params (EndPlayReason :type EEndPlayReason%20Enum :default "Destroyed"))
  ;[0.1302 0.09952 0.03311] EndPlay: Unregister Auto Marking
  (|CallParentFunction EndPlayReason)
  (Networking|SwitchHasAuthority
    (:Authority
      (Class|BMMObjObjectiveActorComponent|UnregisterObjectiveforAutoMarking (Variables|Default|GetB_MMObj_ObjectiveActorComponent)))))

(fn BossArenaAwareness)

(fn ConstructionScript)
