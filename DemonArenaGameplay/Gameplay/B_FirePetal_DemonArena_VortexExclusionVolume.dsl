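; Fire Petal vortex exclusion volume Blueprint for the Demon Arena, which reports actors crossing its Collision component.
; Begin and end overlap events pass the other actor to OnEnteredVolume and OnExitedVolume on its BP_DemonArena_VortexVolumeComponent.
; The Blueprint does no filtering and has an empty construction script, leaving any exclusion behavior to that component.
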(event OnComponentBeginOverlap(Collision)
  (params (OverlappedComponent :type Primitive%20Component%20Object%20Reference) (OtherActor :type Actor%20Object%20Reference) (OtherComp :type Primitive%20Component%20Object%20Reference) (OtherBodyIndex :type Integer) (bFromSweep :type Boolean) (SweepResult :type Hit%20Result%20Structure%20%28by%20ref%29))
  (Class|BPDemonArenaVortexVolumeComponent|OnEnteredVolume (Variables|Default|GetBP_DemonArena_VortexVolumeComponent) OtherActor))

(event OnComponentEndOverlap(Collision)
  (params (OverlappedComponent :type Primitive%20Component%20Object%20Reference) (OtherActor :type Actor%20Object%20Reference) (OtherComp :type Primitive%20Component%20Object%20Reference) (OtherBodyIndex :type Integer))
  (Class|BPDemonArenaVortexVolumeComponent|OnExitedVolume (Variables|Default|GetBP_DemonArena_VortexVolumeComponent) OtherActor))

(fn ConstructionScript)
